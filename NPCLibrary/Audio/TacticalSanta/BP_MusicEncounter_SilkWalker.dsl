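; Music encounter Blueprint for the SilkWalker boss that manages its combat music on top of inherited encounter logic.
; Entering combat casts the boss actor to BP_PlayerPawn_Tandem and attaches the combat music audio component to its mesh.
; Completing the encounter or the boss actor's EndPlay fades that audio out over the configured duration and binds encounter uninitialization to its OnAudioFinished event.
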(event EventOnBossActorEndPlay
  (params (Actor :type Actor%20Object%20Reference) (EndPlayReason :type EEndPlayReason%20Enum :default "Destroyed"))
  (|CallParentFunction Actor EndPlayReason)
  (bind _combat_music_audio_component (Variables|MusicEncounter|Boss|Combat|GetCombat%20Music%20Audio%20Component)
    (:then
      (Audio|Components|Audio|FadeOut _combat_music_audio_component (Variables|MusicEncounter|GetFade%20Out%20Duration) 0.0 "Linear" :!class /Script/Engine.AudioComponent)
      (EventDispatchers|BindEventtoOnAudioFinished _combat_music_audio_component (EventDispatchers|CreateEvent :!event Uninitialize%20Music%20Encounter)))
    (:else)))

(event EventEnterCompletedState
  (|CallParentFunction)
  (bind _combat_music_audio_component (Variables|MusicEncounter|Boss|Combat|GetCombat%20Music%20Audio%20Component)
    (:then
      (Audio|Components|Audio|FadeOut _combat_music_audio_component (Variables|MusicEncounter|GetFade%20Out%20Duration) 0.0 "Linear" :!class /Script/Engine.AudioComponent)
      (EventDispatchers|BindEventtoOnAudioFinished _combat_music_audio_component (EventDispatchers|CreateEvent :!event Uninitialize%20Music%20Encounter)))
    (:else)))

(event EventEnterCombatState
  (|CallParentFunction)
  (bind _combat_music_audio_component (Variables|MusicEncounter|Boss|Combat|GetCombat%20Music%20Audio%20Component)
    (:then
      (bind _boss_actor (Variables|MusicEncounter|Boss|GetBoss%20Actor)
        (:then
          (bind _asbp_player_pawn_tandem (Utilities|Casting|CastToBP_PlayerPawn_Tandem _boss_actor))
          (Transformation|AttachComponentToComponent _combat_music_audio_component (Variables|/Script/Engine.Character|GetMesh _asbp_player_pawn_tandem) "None" "KeepRelative" "KeepRelative" "KeepRelative" true))
        (:else)))
    (:else)))
